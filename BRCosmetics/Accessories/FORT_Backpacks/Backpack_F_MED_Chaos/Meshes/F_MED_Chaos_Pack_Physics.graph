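# Physics asset for the Chaos backpack cosmetic, defining collision and joints for its dangling chains.
# Capsule bodies cover the pack and six small chain bones.
# Five dyn_chain bodies link in series from the pack with limited swing, while a separate middle chain hangs off the pack with tighter, asymmetric swing limits.
# All joints set their twist limit to zero.
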PhysicsAsset /BRCosmetics/Accessories/FORT_Backpacks/Backpack_F_MED_Chaos/Meshes/F_MED_Chaos_Pack_Physics
bodies {
 dyn_chain_1 [{shapeType=Capsule, length=0.2949, radius=0.8959, shapeName=dyn_chain_1_capsule, center=<1.3244 0 0>, rotation=<rot 90 0 0>}]
 dyn_chain_2 [{shapeType=Capsule, length=0.6041, radius=0.8959, shapeName=dyn_chain_2_capsule, center=<1.5 0 0>, rotation=<rot 90 0 0>}]
 dyn_chain_3 [{shapeType=Capsule, length=0.3339, radius=0.8959, shapeName=dyn_chain_3_capsule, center=<1.3578 0 0>, rotation=<rot 90 0 0>}]
 dyn_chain_4 [{shapeType=Capsule, length=0.4085, radius=0.8959, shapeName=dyn_chain_4_capsule, center=<1.3154 0 0>, rotation=<rot 90 0 0>}]
 dyn_chain_5 [{shapeType=Capsule, length=0.6041, radius=0.8959, shapeName=dyn_chain_5_capsule, center=<1.5 0 0>, rotation=<rot 90 0 0>}]
 dyn_middle_chain [{shapeType=Capsule, length=1, radius=1, shapeName=dyn_middle_chain_capsule, center=<1.7971 0 0>, rotation=<rot 90 0 0>}]
 pack [{
   shapeType=Capsule
   length=20.7613
   radius=2.8291
   shapeName=pack_capsule
   center=<-3.9207 -5.6455 -21.7277>
   rotation=<rot 8.9642 -0.0001 -0.0007>
  }]
}
constraints [
 {bone1Name=dyn_chain_1, bone2Name=pack, swing1LimitDegrees=15, swing2LimitDegrees=15, twistLimitDegrees=0}
 {bone1Name=dyn_chain_2, bone2Name=dyn_chain_1, swing1LimitDegrees=15, swing2LimitDegrees=15, twistLimitDegrees=0}
 {bone1Name=dyn_chain_3, bone2Name=dyn_chain_2, swing1LimitDegrees=15, swing2LimitDegrees=15, twistLimitDegrees=0}
 {bone1Name=dyn_chain_4, bone2Name=dyn_chain_3, swing1LimitDegrees=15, swing2LimitDegrees=15, twistLimitDegrees=0}
 {bone1Name=dyn_chain_5, bone2Name=dyn_chain_4, swing1LimitDegrees=15, swing2LimitDegrees=15, twistLimitDegrees=0}
 {bone1Name=dyn_middle_chain, bone2Name=pack, swing1LimitDegrees=10, swing2LimitDegrees=2, twistLimitDegrees=0}
]
